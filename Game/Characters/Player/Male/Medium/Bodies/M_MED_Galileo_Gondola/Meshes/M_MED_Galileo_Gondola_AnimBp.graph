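# Animation Blueprint for the Galileo Gondola male medium player body, adding secondary jacket motion to a pose copied from the parent mesh.
# With rigid body enabled, it simulates physics under skydiving force and overridden gravity, weakened during freefall, then nudges one right-front jacket bone.
# Otherwise, additive bone offsets move both jacket sides, weighted by each thigh's raise measured against the pelvis in the event graph.
# Crouch and jump alphas are computed but unused in the AnimGraph.
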AnimBlueprint /Game/Characters/Player/Male/Medium/Bodies/M_MED_Galileo_Gondola/Meshes/M_MED_Galileo_Gondola_AnimBp
target_skeleton=&/M_MED_Galileo_Gondola_Skeleton
graphs [
 AnimGraph {nodes [
   n1 = RigidBody!(
    ComponentPose=Local To Component(LocalPose=Use cached pose 'PreCache'())
    Alpha
    OverrideWorldGravity:Vector=Get gravity_override() -> (gravity_override:Vector)
    ExternalForce=Break Fort Anim Input Skydiving External Force(FortAnimInput_SkydivingExternalForce=Get SkydivingExternalForceInput()) -> (
     bUseSkydivingVectorForce=~
     SkydivingVectorForce=~
     HeadToPelvisDirection=~
     FloatingMultiplier=~
     FloatingAdditive=~
     DivingMultiplier=~
     DivingAdditive=~
     ParachutingMultiplier=~
     ParachutingAdditive=~
    ).SkydivingVectorForce
   ) -> (Pose) {
    bClampLinearTranslationLimitToRefPose=true
    bOverrideWorldGravity=true
    overridePhysicsAsset=&/M_MED_Galileo_Gondola_Physics
    simulationSpace=BaseBoneSpace
    alphaBoolBlend {blendOption=Linear}
    baseBoneRef {boneName=spine_01}
    componentAppliedLinearAccClamp=<3000 3000 3000>
    componentLinearAccScale=<0.4 0.3 0.3>
    componentLinearVelScale=<0.4 0.3 0.3>
    overrideWorldGravity {z=-980}
   }
   n2 = "Transform (Modify) Bone - Bone: dyn_jacket_r_fr_MSTR"(ComponentPose=@n1, Alpha=Get thigh_up_r() -> (thigh_up_r:double), Translation="0,0,0", Rotation="30,-30,0", Scale="1,1,1") {
    rotationMode=BMM_Additive
    rotationSpace=BCS_BoneSpace
    translationMode=BMM_Additive
    translationSpace=BCS_BoneSpace
    alphaBoolBlend {blendOption=Linear}
    boneToModify {boneName=dyn_jacket_r_fr_MSTR}
    rotation {pitch=30, yaw=-30}
   }
   n3 = Get thigh_up_r()
   n4 = Get thigh_up_l() -> (thigh_up_l:double)
   n5 = "Transform (Modify) Bone - Bone: dyn_jacket_l_fr_01"(ComponentPose=Local To Component(LocalPose=Use cached pose 'PreCache'()), Alpha=@n4, Translation="-2.7,0,0", Rotation="0,0,0", Scale="1,1,1") {
    rotationMode=BMM_Additive
    rotationSpace=BCS_BoneSpace
    translationMode=BMM_Additive
    translationSpace=BCS_BoneSpace
    alphaBoolBlend {blendOption=Linear}
    boneToModify {boneName=dyn_jacket_l_fr_01}
    translation {x=-2.7}
   }
   n6 = "Transform (Modify) Bone - Bone: dyn_jacket_l_fr_02"(ComponentPose=@n5, Alpha=@n4, Translation="-3.2,0,0", Rotation="95,-10,30.000", Scale="1,1,1") {
    rotationMode=BMM_Additive
    rotationSpace=BCS_BoneSpace
    translationMode=BMM_Additive
    translationSpace=BCS_BoneSpace
    alphaBoolBlend {blendOption=Linear}
    boneToModify {boneName=dyn_jacket_l_fr_02}
    rotation=<rot 95 -10 30>
    translation {x=-3.2}
   }
   n7 = "Transform (Modify) Bone - Bone: dyn_jacket_l_01"(ComponentPose=@n6, Alpha=@n4, Translation="-2.5,0,0", Rotation="15,15,-15", Scale="1,1,1") {
    rotationMode=BMM_Additive
    rotationSpace=BCS_BoneSpace
    translationMode=BMM_Additive
    translationSpace=BCS_BoneSpace
    alphaBoolBlend {blendOption=Linear}
    boneToModify {boneName=dyn_jacket_l_01}
    rotation=<rot 15 15 -15>
    translation {x=-2.5}
   }
   n8 = "Transform (Modify) Bone - Bone: dyn_jacket_r_fr_01"(ComponentPose=@n7, Alpha=@n3, Translation="-2.7,0,0", Rotation="0,0,0", Scale="1,1,1") {
    rotationMode=BMM_Additive
    rotationSpace=BCS_BoneSpace
    translationMode=BMM_Additive
    translationSpace=BCS_BoneSpace
    alphaBoolBlend {blendOption=Linear}
    boneToModify {boneName=dyn_jacket_r_fr_01}
    translation {x=-2.7}
   }
   n9 = "Transform (Modify) Bone - Bone: dyn_jacket_r_fr_02"(ComponentPose=@n8, Alpha=@n3, Translation="-3.2,0,0", Rotation="95,10,-30", Scale="1,1,1") {
    rotationMode=BMM_Additive
    rotationSpace=BCS_BoneSpace
    translationMode=BMM_Additive
    translationSpace=BCS_BoneSpace
    alphaBoolBlend {blendOption=Linear}
    boneToModify {boneName=dyn_jacket_r_fr_02}
    rotation=<rot 95 10 -30>
    translation {x=-3.2}
   }
   n10 = "Transform (Modify) Bone - Bone: dyn_jacket_r_01"(ComponentPose=@n9, Alpha=@n3, Translation="-2.5,0,0", Rotation="15,-15,15", Scale="1,1,1") {
    rotationMode=BMM_Additive
    rotationSpace=BCS_BoneSpace
    translationMode=BMM_Additive
    translationSpace=BCS_BoneSpace
    alphaBoolBlend {blendOption=Linear}
    boneToModify {boneName=dyn_jacket_r_01}
    rotation=<rot 15 -15 15>
    translation {x=-2.5}
   }
   Output Pose(Result=Blend Poses by bool(BlendPose_0=Component To Local(ComponentPose=@n2), BlendPose_1=Component To Local(ComponentPose=@n10), bActiveValue=Get bRigidBodyEnabled()) {blendType=Linear}) {layerGroup=None}
   n11 = Copy Pose From Mesh() {bUseAttachedParent=true}
   Save cached pose 'PreCache'(Pose=Component To Local(ComponentPose=Local To Component(LocalPose=@n11))) {cacheName=PreCache}
  ]}
 EventGraph {nodes [
   Event BlueprintUpdateAnimation()
   > n1 = Sequence!(execute) -> (then_0, then_1, then_2:Exec, then_3:Exec)
   > Set crouching alpha(crouching alpha:double="To Float (Boolean)"(InBool=Get bIsCrouching())) -> (Output_Get:double)
   > n2 = Branch(Condition=Get bIsJumping())
   Set jump_alpha(execute=[@n2.then, @n3.then], jump_alpha:double=1) -> (Output_Get:double)
   n3 = Branch(execute=@n2.else, Condition=Get bIsFalling())
   Set jump_alpha(execute=@n3.else, jump_alpha=0)
   n4 = BreakTransform(InTransform=GetDeltaTransformFromRefPose(self=GetOwnerSkeletalMeshComponent(), BoneName=thigh_r, BaseName=pelvis))
   Set thigh_up_r(execute=@n1.then_3, thigh_up_r:double=MapRangeClamped(Value=BreakRotator(InRot=@n4.Rotation).Yaw, InRangeA=-20, InRangeB=-100)) -> (Output_Get:double)
   n5 = BreakTransform(InTransform=GetDeltaTransformFromRefPose(self=GetOwnerSkeletalMeshComponent(), BoneName=thigh_l, BaseName=pelvis))
   Set thigh_up_l(execute=@n1.then_2, thigh_up_l:double=MapRangeClamped(Value=BreakRotator(InRot=@n5.Rotation).Yaw, InRangeA=-20, InRangeB=-100)) -> (Output_Get:double)
   n6 = Branch(execute=@n1.then_1, Condition=Get bIsSkyDiving())
   > n7 = Branch(Condition=Get bIsParachuteOpened())
   Set gravity_override(execute=[@n7.then, @n6.else], gravity_override="0, 0,-980")
   Set gravity_override(execute=@n7.else, gravity_override="0, 0,-300")
   TryGetPawnOwner()
  ]}
]
variables [
 crouching alpha {type=number}
 gravity_override {type=Vector}
 jump_alpha {type=number}
 thigh_up_l {type=number}
 thigh_up_r {type=number}
]
